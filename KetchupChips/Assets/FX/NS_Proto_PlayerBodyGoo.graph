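# Prototype Niagara system that attaches goo and mud to the player's body via skeletal socket positions.
# Mud_Body_Decal places six socket-following meshes whose material parameters come from Blueprint vectors, with one boosted while sliding or crouched.
# GooMesh001 adds socket-attached blobs, while GooDrops spawns drips using a lengthening pendulum constraint, switched by normalized age, plus owner-velocity drag.
# Goo spread and fades drive blob scale and drip rate; two older-style GooMesh emitters are disabled.
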NiagaraSystem /KetchupChips/Assets/FX/NS_Proto_PlayerBodyGoo
dependencies {
 usedDataInterfaces [
  &NiagaraDataInterfaceArrayFloat
  &NiagaraDataInterfaceCurve
  &NiagaraDataInterfaceArrayInt32
  &NiagaraDataInterfaceSocketReader
 ]
 usedDynamicInputs [&ClampFloat, &Frac_Float]
 usedModules [
  &self:SystemSpawnScript.SystemScriptSource.SystemScriptGraph.NiagaraNodeAssignment_0.SetVariables_#1
  &self:SystemSpawnScript.SystemScriptSource.SystemScriptGraph.NiagaraNodeAssignment_2.SetVariables_#2
  &self:SystemSpawnScript.SystemScriptSource.SystemScriptGraph.NiagaraNodeAssignment_1.SetVariables_#3
  &SystemState
  &self:SystemSpawnScript.SystemScriptSource.SystemScriptGraph.NiagaraNodeAssignment_5.SetVariables_#4
  &self:SystemSpawnScript.SystemScriptSource.SystemScriptGraph.NiagaraNodeAssignment_3.SetVariables_#5
  &self:SystemSpawnScript.SystemScriptSource.SystemScriptGraph.NiagaraNodeAssignment_4.SetVariables_#6
  &Timeline
  &self:SystemSpawnScript.SystemScriptSource.SystemScriptGraph.NiagaraNodeAssignment_6.SetVariables_#7
  &self:GetCharacterScale
 ]
}
emitters {
 Mud_Body_Decal {
  emitterUpdateScript [
   &EmitterState()
   &SpawnBurst_Instantaneous(Spawn Count=6, Use Spawn Probability=false, UseLoopCountLimit=true)
   &SpawnBurst_Instantaneous(Spawn Count=6, Spawn Time=2.2833, Use Spawn Probability=false, UseLoopCountLimit=false) {module=SpawnBurst_Instantaneous001}
  ]
  particleSpawnScript [
   &PartitionParticles(Number Of Partitions=6)
   &InitializeParticle(
    Lifetime=1000000000
    Mesh Scale Mode=::NewEnumerator3
    Mesh Uniform Scale=&SelectFloatFromArray(
     Direct Array Index=&Add_Integer(A=0, B=$Particles.PartitionParticles.Partition)
     Float Selection Array=DataInterface() {propertyValues {FloatData [25,25,25,25,20,20]}}
    )
    UsePositionOffset=false
    Write Material Random=true
   )
   &CameraOffset(Camera Offset Amount=30)
   /SetVariables() {isSetParameters=true}
   &DynamicMaterialParameters(
    Index 0 Param 1=&Add_Float(A=&MakeFloatFromVector4(Vector 4=$User.InputFromBP1, Channel=::NewEnumerator0), B=0)
    Index 0 Param 2=&MakeFloatFromVector4(Vector 4=$User.InputFromBP1, Channel=::NewEnumerator1)
    Index 0 Param 3=&Add_Float(A=&MakeFloatFromVector4(Vector 4=$User.InputFromBP1, Channel=::NewEnumerator2), B=0)
    Index 0 Param 4=&MakeFloatFromVector4(Vector 4=$User.InputFromBP1, Channel=::NewEnumerator3)
    Index 1 Param 1=&MakeFloatFromVector4(Vector 4=$User.InputFromBP2, Channel=::NewEnumerator0)
    Index 1 Param 2=&MakeFloatFromVector4(Vector 4=$User.InputFromBP2, Channel=::NewEnumerator1)
    Index 1 Param 3=&MakeFloatFromVector4(Vector 4=$User.InputFromBP2, Channel=::NewEnumerator2)
    Module.Param0WriteEnabled=true
    Module.Param0WriteEnabled1=true
    Module.Param1WriteEnabled=true
    Module.Param1WriteEnabled1=true
    Module.Param2WriteEnabled=true
    Module.Param2WriteEnabled1=true
    Module.Param3WriteEnabled=true
    Module.Param3WriteEnabled1=false
    Write Parameter Index 1=true
   ) {module=DynamicMaterialParameters001}
  ]
  particleUpdateScript [
   &ParticleState()
   &SocketLocation(
    Orientation Sampling=::NewEnumerator1
    SamplePosOffset=true
    Sampled Position Offset=&SelectVectorFromArray(
     Direct Array Index=$Particles.PartitionParticles.Partition
     Vector Selection Array=DataInterface() {propertyValues {FloatData [<0 9 0>, <0 -9 0>, <-20 0 0>, <20 0 0>, <-30 0 0>, <30 0 0>]}}
    )
    Socket Index=$Particles.PartitionParticles.Partition
    Socket Reader=$System.Socket Reader
    Socket Sampling Mode=::NewEnumerator3
    VelocitySpeedLimit=false
   )
   &DynamicMaterialParameters(
    Index 0 Param 1=&MakeFloatFromVector4(Vector 4=$User.InputFromBP1, Channel=::NewEnumerator0)
    Index 0 Param 2=&MakeFloatFromVector4(Vector 4=$User.InputFromBP1, Channel=::NewEnumerator1)
    Index 0 Param 3=&MakeFloatFromVector4(Vector 4=$User.InputFromBP1, Channel=::NewEnumerator2)
    Index 0 Param 4=&MakeFloatFromVector4(Vector 4=$User.InputFromBP1, Channel=::NewEnumerator3)
    Index 1 Param 1=&Multiply_Float(
     A=&MakeFloatFromVector4(Vector 4=$User.InputFromBP2, Channel=::NewEnumerator0)
     B=&MakeCustomFloatFromBool(Bool=&Boolean_OrOperation(Bool1=$User.IsSliding, Bool2=$User.IsCrouched), True Float=1.5, False Float=1)
    )
    Index 1 Param 2=&Multiply_Float(A=&MakeFloatFromVector4(Vector 4=$User.InputFromBP2, Channel=::NewEnumerator1), B=1)
    Index 1 Param 3=&MakeFloatFromVector4(Vector 4=$User.InputFromBP2, Channel=::NewEnumerator2)
    Module.Param0WriteEnabled=true
    Module.Param0WriteEnabled1=true
    Module.Param1WriteEnabled=false
    Module.Param1WriteEnabled1=true
    Module.Param2WriteEnabled=true
    Module.Param2WriteEnabled1=true
    Module.Param3WriteEnabled=false
    Module.Param3WriteEnabled1=false
    Write Parameter Index 1=true
   )
   /SetVariables(Particles.Counter:NiagaraFloat=&Add_Float(A=$Particles.Counter, B=$Engine.DeltaTime)) {isSetParameters=true}
  ]
  rendererClasses [&NiagaraMeshRendererProperties]
  renderers [{rendererClass=&NiagaraMeshRendererProperties, rendererIndex=0}]
 }
 NE_GooDrops {
  emitterUpdateScript [
   &EmitterState()
   &SpawnRate(
    SpawnRate=&Multiply_Float(
     A=&Lerp_Float(A=2, B=8, Alpha=$System.NormalizedGooSpread)
     B=&Multiply_Float(A=$System.FadeIn, B=$System.FadeOut)
    )
    Use Spawn Probability=false
   )
  ]
  particleSpawnScript [
   &InitializeParticle(
    Lifetime=&RandomRangeFloat(Minimum=2.5, Maximum=3.5)
    Mesh Scale=&Multiply_VectorByFloat(Vector=<0.3 0.3 0.8>, Float=&RandomRangeFloat(Minimum=0.2, Maximum=0.25))
    Mesh Scale Mode=::NewEnumerator4
    Sprite Size Mode=::NewEnumerator3
    Uniform Sprite Size=5
    UsePositionOffset=false
    Write Material Random=true
   )
   &PartitionParticles(Number Of Partitions=4, Partition Mode:ENiagara_PartitionMode_v2=::NewEnumerator5)
   &InitialMeshOrientation(Mesh Orientation Mode=::NewEnumerator7, Use Rotation Vector=true)
  ]
  particleUpdateScript [
   &SocketLocation(
    Orientation Sampling=::NewEnumerator1
    Position Sampling=::NewEnumerator1
    SamplePosOffset=true
    Socket Index=$Particles.PartitionParticles.Partition
    Socket Reader=$System.Socket Reader
    Socket Sampling Mode=::NewEnumerator3
    VelocitySpeedLimit=false
   )
   &ShapeLocation(
    Additional Quaternion Rotation=true
    Additional Yaw / Pitch / Roll=<0 0.25 0>
    Apply To Particle Position=::NewEnumerator1
    Cylinder Height=20
    Cylinder Radius=5
    MaskSpawnGroup=false
    Override Seed=false
    Rotation Mode=::NewEnumerator3
    Rotation Quaternion=$Output.SocketLocation.SampledRotation
    Shape Origin=$Output.SocketLocation.SampledPosition
    Shape Primitive=::NewEnumerator1
    Surface Only=false
    UseOwnerScale=true
   )
   /SetVariables(Particles.Connection:NiagaraBool=&SetBoolByFloatComparison(A=$Particles.NormalizedAge, B=0.7)) {isSetParameters=true}
   &ParticleState()
   &ScaleMeshSize(Scale Factor=&VectorFromFloat(Value=&FloatFromCurve(FloatCurve=DataInterface() {propertyValues {Curve {keys [
          {interpMode=RCIM_Cubic, tangentMode=RCTM_User, time=0, value=0}
          {interpMode=RCIM_Cubic, tangentMode=RCTM_User, time=0.8, value=1}
          {arriveTangent=-0.2019, interpMode=RCIM_Cubic, leaveTangent=-0.2019, tangentMode=RCTM_User, time=0.95, value=1}
          <at 1 0 RCIM_Cubic>
         ]}}})))
   &GravityForce(Gravity=<0 0 -600>)
   &PendulumSetup(
    Module.Enable Spring Driver=false
    Pendulum Length:NiagaraFloat=&FloatFromCurve(
     FloatCurve=DataInterface() {propertyValues {Curve {keys [<at 0 0 RCIM_Cubic>, {arriveTangent=0.0714, interpMode=RCIM_Cubic, leaveTangent=0.0714, tangentMode=RCTM_User, time=0.7, value=1}]}}}
     Scale Curve=3
    )
    Pendulum Pivot:NiagaraPosition=$Particles.ShapeLocation.ShapePosition
   )
   &AccelerationForce(Acceleration=&Multiply_VectorByFloat(
     Vector=&Multiply_Vector(A=$Engine.Owner.Velocity, B=<-1 -1 0>)
     Float=&FloatFromCurve(FloatCurve=DataInterface() {propertyValues {Curve {keys [{interpMode=RCIM_Cubic, tangentMode=RCTM_User, time=0.6, value=0}, {interpMode=RCIM_Cubic, tangentMode=RCTM_User, time=0.9, value=1}]}}})
    ))
   &SolveForcesAndVelocity(Clamp Velocity=false, Limit Acceleration=false)
   /SetVariables(Particles.MeshOrientation=&FindQuatBetweenTwoVectors(
     From Vector=<0 0 -1>
     To Vector=&NormalizeVector(
      Vector To Normalize=&SubtractVectorFromPosition(Position=$Particles.Position, Vector=Linked() {linkedVariable {name=Particles.ShapeLocation.ShapePosition, type=NiagaraPosition}})
      Safe Normalize=true
      Safe Fallback Threshold=0
      Safe Fallback Vector=<0 0 0.001>
     )
    )) {isSetParameters=true}
   /SetVariables(Particles.PositionTemp:NiagaraPosition=$Particles.Position) {isSetParameters=true}
   &PendulumConstraint()
   /SetVariables(Particles.Position=&MakeCustomVectorFromBool(
     Bool=$Particles.Connection
     True Vector=Linked() {linkedVariable {name=Particles.PositionTemp, type=NiagaraPosition}}
     False Vector=Linked() {linkedVariable {name=Particles.Position, type=NiagaraPosition}}
    )) {isSetParameters=true}
   &CalculateAccurateVelocity()
   &DynamicMaterialParameters(Module.Index0Float0Write=true, Module.Index0Float1Write=true, Module.Index0Float2Write=true, Module.Index0Float3Write=true)
   /SetVariables(Particles.Position=&MakeCustomVectorFromBool(
     Bool=$Particles.Connection
     True Vector=Linked() {linkedVariable {name=Particles.Position, type=NiagaraPosition}}
     False Vector=$Transient.PhysicsPendulumPivot
    )) {isSetParameters=true}
  ]
  rendererClasses [&NiagaraMeshRendererProperties]
  renderers [{rendererClass=&NiagaraMeshRendererProperties, rendererIndex=0}]
 }
 NE_GooMesh {
  bEnabled=false
  emitterUpdateScript [
   &EmitterState(
    Inactive Response=::NewEnumerator1
    Life Cycle Mode=::NewEnumerator1
    Loop Behavior=::NewEnumerator1
    Loop Duration Mode=::NewEnumerator1
    UseLoopDelay=false
   )
   &SpawnBurst_Instantaneous(Spawn Count=$System.GooMeshNumber, Use Spawn Probability=false, UseLoopCountLimit=false)
  ]
  particleSpawnScript [
   &PartitionParticles(Number Of Partitions=$System.GooMeshNumber)
   &InitializeParticle(
    Lifetime=5
    Mesh Scale=&Multiply_VectorByFloat(
     Vector=&SelectVectorFromArray(
      Direct Array Index=$Particles.PartitionParticles.Partition
      Vector Selection Array=DataInterface() {propertyValues {FloatData [<0 0 0>, <0 0 0>, <1.5 0.8 0.8>, <1.5 0.8 0.8>, <1.5 1 1>, <1.5 1 1>]}}
     )
     Float=0.08
    )
    Mesh Scale Mode=::NewEnumerator4
    UsePositionOffset=false
    Write Material Random=true
   )
   &InitialMeshOrientation(Mesh Orientation Mode=::NewEnumerator7, Use Rotation Vector=true)
  ]
  particleUpdateScript [
   &ParticleState(Kill Particles When Lifetime Has Elapsed=false)
   &ScaleMeshSize(Scale Factor=&Multiply_Vector(
     A=&VectorFromFloat(Value=&Lerp_Float(A=0.6, B=0.9, Alpha=$System.NormalizedGooSpread))
     B=$System.Adjust Scale by Character Size
    ))
   &SocketLocation(
    Orientation Sampling=::NewEnumerator1
    Position Sampling=::NewEnumerator1
    SamplePosOffset=true
    Sampled Position Offset=&SelectVectorFromArray(
     Direct Array Index=$Particles.PartitionParticles.Partition
     Vector Selection Array=DataInterface() {propertyValues {FloatData [
        <-3 3 1>
        <3 -3 1.5>
        <-15 -3 1>
        <15 4 2>
        <-15 -2 -1>
        <15 2 2>
        <7 -2 -2>
        <-7 2 -0.5>
        <3 -0.5 0>
        <-3 0.5 0>
        <-2 -2 0>
        <0 -1 -2.5>
        <0 -2.5 0>
        <4 -1 0>
       ]}}
    )
    Socket Index=$Particles.PartitionParticles.Partition
    Socket Reader=$System.Socket Reader
    Socket Sampling Mode=::NewEnumerator3
    VelocitySpeedLimit=false
   )
   /SetVariables(Particles.BoneDirection:Vector3f=&RotateVectorByQuaternion(
     Vector To Rotate=&SelectVectorFromArray(
      Direct Array Index=$Particles.PartitionParticles.Partition
      Vector Selection Array=DataInterface() {propertyValues {FloatData [<-1 0 0>, <1 0 0>, <-1 0 0>, <1 0 0>, <-1 0 0>, <1 0 0>, <1 0 0>, <-1 0 0>, <1 0 0>, <-1 0 0>, <1 0 0>, <-1 0 0>, <-1 0 0>, <-1 0 0>]}}
     )
     Quaternion=$Output.SocketLocation.SampledRotation
    )) {isSetParameters=true, module=SetVariables_#8}
   /SetVariables(Particles.MeshOrientation=&MultiplyQuaternion(
     Quaternion A=&MultiplyQuaternion(
      Quaternion A=$Particles.Initial.MeshOrientation
      Quaternion B=&OrientMeshToVector(Look At Direction=$Particles.BoneDirection, Reference Axis Vector=<1 0 0>)
     )
     Quaternion B=&MakeQuaternion(
      Quaternion From=0
      Axis=&SelectVectorFromArray(
       Direct Array Index=$Particles.PartitionParticles.Partition
       Vector Selection Array=DataInterface() {propertyValues {FloatData [<0 1 0>, <0 1 0>, <1 0 0>, <1 0 0>, <1 0 0>, <1 0 0>]}}
      )
      Angle=&RandomRangeFloat(Minimum=0, Maximum=1)
      From=?
      To=?
      Angle Type=::NewEnumerator1
      Normalize Vectors=false
     )
    )) {isSetParameters=true, module=SetVariables_#9}
   /SetVariables(Particles.Position=&AddVectorToPosition(
     Position=$Output.SocketLocation.SampledPosition
     Vector=&Multiply_VectorByFloat(
      Vector=$Particles.BoneDirection
      Float=&SelectFloatFromArray(
       Direct Array Index=$Particles.PartitionParticles.Partition
       Float Selection Array=DataInterface() {propertyValues {FloatData [0,0,0,0,0,0,0,0,0,0,0,0]}}
      )
     )
    )) {isSetParameters=true, module=SetVariables_#10}
   &DynamicMaterialParameters(
    Index 0 Param 4=&Multiply_Float(A=$Emitter.Age, B=0.03)
    Index 1 Param 4=$Particles.MaterialRandom
    Module.Index0Float0Write=true
    Module.Index0Float1Write=true
    Module.Index0Float2Write=true
    Module.Index0Float3Write=true
    Module.Index1Float0Write=true
    Module.Index1Float1Write=true
    Module.Index1Float2Write=true
    Module.Index1Float3Write=true
    Module.Index2Float0Write=true
    Module.Index2Float1Write=true
    Module.Index2Float2Write=true
    Module.Index2Float3Write=true
    Write Parameter Index 1=::NewEnumerator1
    Write Parameter Index 2=::NewEnumerator1
   )
  ]
  rendererClasses [&NiagaraMeshRendererProperties]
  renderers [{rendererClass=&NiagaraMeshRendererProperties, rendererIndex=0}]
 }
 NE_GooMesh001 {
  emitterUpdateScript [
   &EmitterState(
    Inactive Response=::NewEnumerator1
    Life Cycle Mode=::NewEnumerator1
    Loop Behavior=::NewEnumerator1
    Loop Duration Mode=::NewEnumerator1
    UseLoopDelay=false
   )
   &SpawnBurst_Instantaneous(Spawn Count=$System.GooMeshNumber, Use Spawn Probability=false, UseLoopCountLimit=false)
  ]
  particleSpawnScript [
   &PartitionParticles(Number Of Partitions=$System.GooMeshNumber)
   &InitializeParticle(
    Lifetime=5
    Mesh Scale=&Multiply_VectorByFloat(
     Vector=&SelectVectorFromArray(
      Direct Array Index=$Particles.PartitionParticles.Partition
      Vector Selection Array=DataInterface() {propertyValues {FloatData [<1 1 1>, <1 1 1>, <0.7 0.7 1>, <0.7 0.7 1>, <0.7 0.7 1>, <0.7 0.7 1>]}}
     )
     Float=1
    )
    Mesh Scale Mode=::NewEnumerator4
    UsePositionOffset=false
    Write Material Random=true
   )
   &InitialMeshOrientation(
    Mesh Orientation Mode=::NewEnumerator7
    Rotation=&SelectVectorFromArray(
     Direct Array Index=$Particles.PartitionParticles.Partition
     Vector Selection Array=DataInterface() {propertyValues {FloatData [<0 0.5 0>, <0.5 0 0>, <0 0.25 0>, <0 -0.25 0>, <0 0.25 0>, <0 -0.25 0>]}}
    )
    Use Rotation Vector=true
   )
  ]
  particleUpdateScript [
   &ParticleState(Kill Particles When Lifetime Has Elapsed=false)
   &ScaleMeshSize(Scale Factor=&Multiply_Vector(
     A=&MakeVector(
      X=&Multiply_Float(A=$System.FadeIn, B=$System.FadeOut)
      Y=&Multiply_Float(A=$System.FadeIn, B=$System.FadeOut)
      Z=1
     )
     B=&VectorFromFloat(Value=&Lerp_Float(A=0.6, B=0.9, Alpha=$System.NormalizedGooSpread))
    ))
   &SocketLocation(
    SamplePosOffset=true
    Sampled Position Offset=&SelectVectorFromArray(
     Direct Array Index=$Particles.PartitionParticles.Partition
     Vector Selection Array=DataInterface() {propertyValues {FloatData [<9 6 0>, <-9 -6 0>, <-15 -3 0>, <15 3 0>, <-15 -1.5 0>, <15 1.5 0>]}}
    )
    Socket Index=$Particles.PartitionParticles.Partition
    Socket Reader=$System.Socket Reader
    Socket Sampling Mode=::NewEnumerator3
    VelocitySpeedLimit=false
   )
   /SetVariables(Particles.MeshIndex=&SelectIntFromArray(Direct Array Index=$Particles.PartitionParticles.Partition, Int Selection Array=DataInterface() {propertyValues {IntData [2,2,1,1,0,0]}})) {isSetParameters=true}
   &DynamicMaterialParameters(
    Index 0 Param 4=&Multiply_Float(A=$Emitter.Age, B=0.03)
    Index 1 Param 4=$Particles.MaterialRandom
    Module.Index0Float0Write=true
    Module.Index0Float1Write=true
    Module.Index0Float2Write=true
    Module.Index0Float3Write=true
    Module.Index1Float0Write=true
    Module.Index1Float1Write=true
    Module.Index1Float2Write=true
    Module.Index1Float3Write=true
    Module.Index2Float0Write=true
    Module.Index2Float1Write=true
    Module.Index2Float2Write=true
    Module.Index2Float3Write=true
    Write Parameter Index 1=::NewEnumerator1
    Write Parameter Index 2=::NewEnumerator1
   )
  ]
  rendererClasses [&NiagaraMeshRendererProperties]
  renderers [{rendererClass=&NiagaraMeshRendererProperties, rendererIndex=0}]
 }
 NE_GooMesh002 {
  bEnabled=false
  emitterUpdateScript [
   &EmitterState(
    Inactive Response=::NewEnumerator1
    Life Cycle Mode=::NewEnumerator1
    Loop Behavior=::NewEnumerator1
    Loop Duration Mode=::NewEnumerator1
    UseLoopDelay=false
   )
   &SpawnBurst_Instantaneous(Spawn Count=$System.GooMeshNumber, Use Spawn Probability=false, UseLoopCountLimit=false)
  ]
  particleSpawnScript [
   &PartitionParticles(Number Of Partitions=$System.GooMeshNumber)
   &InitializeParticle(
    Lifetime=5
    Mesh Scale=&Multiply_VectorByFloat(
     Vector=&SelectVectorFromArray(
      Direct Array Index=$Particles.PartitionParticles.Partition
      Vector Selection Array=DataInterface() {propertyValues {FloatData [<1 0.7 0.7>, <1 0.7 0.7>, <1 1 1>, <1 1 1>, <1 1 1>, <1.25 1 1>]}}
     )
     Float=0.128
    )
    Mesh Scale Mode=::NewEnumerator4
    UsePositionOffset=false
    Write Material Random=true
   )
   &InitialMeshOrientation(
    Mesh Orientation Mode=::NewEnumerator7
    Rotation=&SelectVectorFromArray(
     Direct Array Index=$Particles.PartitionParticles.Partition
     Vector Selection Array=DataInterface() {propertyValues {FloatData [<0 0 0.26>, <0 0 0.26>, <0 0.14 0>, <0 0.14 0>, <0 0.14 0>, <0 0.14 0>]}}
    )
    Use Rotation Vector=true
   )
  ]
  particleUpdateScript [
   &ParticleState(Kill Particles When Lifetime Has Elapsed=false)
   &ScaleMeshSize(Scale Factor=&Multiply_Vector(
     A=&VectorFromFloat(Value=&Lerp_Float(A=0.6, B=1, Alpha=$System.NormalizedGooSpread))
     B=$System.Adjust Scale by Character Size
    ))
   &SocketLocation(
    Orientation Sampling=::NewEnumerator1
    Position Sampling=::NewEnumerator1
    SamplePosOffset=true
    Sampled Position Offset=&SelectVectorFromArray(
     Direct Array Index=$Particles.PartitionParticles.Partition
     Vector Selection Array=DataInterface() {propertyValues {FloatData [<0 -5 0>, <0 5 0>, <0 0 0>, <0 0 0>, <0 0 0>, <0 0 0>]}}
    )
    Socket Index=$Particles.PartitionParticles.Partition
    Socket Reader=$System.Socket Reader
    Socket Sampling Mode=::NewEnumerator3
    VelocitySpeedLimit=false
   )
   /SetVariables(Particles.BoneDirection:Vector3f=&RotateVectorByQuaternion(
     Vector To Rotate=&SelectVectorFromArray(
      Direct Array Index=$Particles.PartitionParticles.Partition
      Vector Selection Array=DataInterface() {propertyValues {FloatData [<-1 0 0>, <1 0 0>, <-1 0 0>, <1 0 0>, <-1 0 0>, <1 0 0>, <1 0 0>, <-1 0 0>, <1 0 0>, <-1 0 0>, <1 0 0>, <-1 0 0>, <-1 0 0>, <-1 0 0>]}}
     )
     Quaternion=$Output.SocketLocation.SampledRotation
    )) {isSetParameters=true, module=SetVariables_#8}
   /SetVariables(Particles.MeshOrientation=&MultiplyQuaternion(
     Quaternion A=&MultiplyQuaternion(
      Quaternion A=$Particles.Initial.MeshOrientation
      Quaternion B=&OrientMeshToVector(Look At Direction=$Particles.BoneDirection, Reference Axis Vector=<1 0 0>)
     )
     Quaternion B=&MakeQuaternion(
      Quaternion From=0
      Axis=&SelectVectorFromArray(
       Direct Array Index=$Particles.PartitionParticles.Partition
       Vector Selection Array=DataInterface() {propertyValues {FloatData [<0 1 0>, <0 1 0>, <1 0 0>, <1 0 0>, <1 0 0>, <1 0 0>]}}
      )
      Angle=&RandomRangeFloat(Minimum=0, Maximum=0)
      From=?
      To=?
      Angle Type=::NewEnumerator1
      Normalize Vectors=false
     )
    )) {isSetParameters=true, module=SetVariables_#9}
   /SetVariables(Particles.Position=&AddVectorToPosition(
     Position=$Output.SocketLocation.SampledPosition
     Vector=&Multiply_VectorByFloat(
      Vector=$Particles.BoneDirection
      Float=&SelectFloatFromArray(
       Direct Array Index=$Particles.PartitionParticles.Partition
       Float Selection Array=DataInterface() {propertyValues {FloatData [0,0,0,0,0,0,0,0,0,0,0,0]}}
      )
     )
    )) {isSetParameters=true, module=SetVariables_#10}
   &DynamicMaterialParameters(
    Index 0 Param 4=&Multiply_Float(A=$Emitter.Age, B=0.03)
    Index 1 Param 4=$Particles.MaterialRandom
    Module.Index0Float0Write=true
    Module.Index0Float1Write=true
    Module.Index0Float2Write=true
    Module.Index0Float3Write=true
    Module.Index1Float0Write=true
    Module.Index1Float1Write=true
    Module.Index1Float2Write=true
    Module.Index1Float3Write=true
    Module.Index2Float0Write=true
    Module.Index2Float1Write=true
    Module.Index2Float2Write=true
    Module.Index2Float3Write=true
    Write Parameter Index 1=::NewEnumerator1
    Write Parameter Index 2=::NewEnumerator1
   )
  ]
  rendererClasses [&NiagaraMeshRendererProperties]
  renderers [{rendererClass=&NiagaraMeshRendererProperties, rendererIndex=0}]
 }
}
summary {
 systemName=NS_Proto_PlayerBodyGoo
 emitters [
  {bEnabled=false, emitterName=NE_GooMesh, rendererClasses [&NiagaraMeshRendererProperties]}
  {emitterName=NE_GooDrops, rendererClasses [&NiagaraMeshRendererProperties]}
  {emitterName=NE_GooMesh001, rendererClasses [&NiagaraMeshRendererProperties]}
  {emitterName=Mud_Body_Decal, rendererClasses [&NiagaraMeshRendererProperties]}
  {bEnabled=false, emitterName=NE_GooMesh002, rendererClasses [&NiagaraMeshRendererProperties]}
 ]
 userVariables [
  User.CharacterSizeID {type=NiagaraInt32}
  User.FadeOut {type=NiagaraBool}
  User.GooSpread {type=NiagaraFloat, defaultValue=0.6, description=""}
  User.InputFromBP1 {type=Vector4f, description="", defaultValue {struct=Vector4f, value=<0 0.2 10000 2>}}
  User.InputFromBP2 {type=Vector4f, description="", defaultValue {struct=Vector4f, value=<0 0 0 0>}}
  User.IsCrouched {type=NiagaraBool}
  User.IsSliding {type=NiagaraBool}
  User.Mesh Scale Array {
   type=NiagaraDataInterfaceArrayFloat3
   description=""
   defaultValue {
    struct=NiagaraExt_VariableValue_DataInterface
    value {
     dataInterface=&self:NiagaraDataInterfaceArrayFloat3_0
     dataInterfaceClass=&NiagaraDataInterfaceArrayFloat3
    }
   }
  }
 ]
}
