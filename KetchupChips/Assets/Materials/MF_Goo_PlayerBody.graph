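# Material function for an animated, emissive goo player body, with switchable purple, blue, green and yellow color schemes.
# Layers projected bubbles, caustics, bump-offset particles, fresnel and light-scattering masks, using pre-skinned position and normals.
# A mask decodes vertex-color spherical harmonics along a normal-distorted view direction, with cheaper quality and shading-path fallbacks.
# Outputs material attributes with blended detail normals, while a static switch optionally blends in authored textures through FXMask.
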MaterialFunction /KetchupChips/Assets/Materials/MF_Goo_PlayerBody
graph {nodes [
  CollectionParameter() {
   collection=&FortniteMaterialParameters
   parameterName=SunlightDirectionalVector
  }
  CollectionParameter() {
   collection=&FortniteMaterialParameters
   parameterName=SunlightDirectionalVector
  }
  n1 = TextureObject() {samplerType=SAMPLERTYPE_LinearGrayscale, texture=/Game/Effects/Fort_Effects/Textures/Tile/T_Caustic_Wind_Diagonal}
  n2 = ScalarParameter() {defaultValue=256, group=BugGoo, parameterName=CausticsSize}
  n3 = Multiply(
   A=MaterialFunctionCall!(
    TextureObject=@n1
    TextureSize=@n2
    WorldPosition=NamedRerouteUsage()
    Export Float 4
    In=Constant2Vector() {g=0.02, r=0.02}
    World Space Normal=NamedRerouteUsage()
    ProjectionTransitionContrast
   ) {materialFunction=/Game/Athena/SafeZone/WorldAlignedTexture_ScalePan}.XY Texture
   B=MaterialFunctionCall(
    TextureObject=@n1
    TextureSize=@n2
    WorldPosition=NamedRerouteUsage()
    In=Constant2Vector() {g=0.04, r=-0.04}
    World Space Normal=NamedRerouteUsage()
   ) {materialFunction=&WorldAlignedTexture_ScalePan}.XY Texture
  )
  ComponentMask(None=@n3) {b=true}
  n4 = Constant3Vector()
  n5 = TextureSampleParameter2D() {group=BugGoo, parameterName=Diffuse, texture=&T_BlacksRGB}
  n6 = TextureSampleParameter2D() {group=BugGoo, parameterName=FXMask, texture=&T_BlacksRGB}
  n7 = StaticBoolParameter() {group=BugGoo, parameterName=Use FXmask and textures}
  n8 = ScalarParameter() {defaultValue=0.5, group=BugGoo, parameterName=Specular}
  n9 = TextureSampleParameter2D() {group=BugGoo, parameterName=Specular, samplerType=SAMPLERTYPE_Masks, texture=&T_BlackMask}
  n10 = ScalarParameter() {defaultValue=0.2, group=BugGoo, parameterName=Roughness}
  n11 = NamedRerouteUsage()
  n12 = Constant()
  n13 = TextureCoordinate()
  n14 = SetMaterialAttributes!(
   MaterialAttributes
   BaseColor=StaticSwitch(True=LinearInterpolate(A=@n4, B=@n5.RGB, Alpha=@n6.R), False=@n4, Value=@n7)
   Specular=StaticSwitch(True=LinearInterpolate(A=@n8, B=@n9.R, Alpha=@n6.R), False=@n8, Value=@n7)
   Roughness=StaticSwitch(True=LinearInterpolate(A=@n10, B=@n9.B, Alpha=@n6.R), False=@n10, Value=@n7)
   EmissiveColor=StaticSwitch(True=LinearInterpolate(A=@n11, B=@n12, Alpha=@n6.R), False=@n11, Value=@n7)
   Normal=MaterialFunctionCall(
    BaseNormal=TextureSampleParameter2D(UVs=@n13) {group=BugGoo, parameterName=Normals, samplerType=SAMPLERTYPE_Normal, texture=/Engine/EngineMaterials/FlatNormal}.RGB
    AdditionalNormal=NamedRerouteUsage()
   ) {materialFunction=&BlendAngleCorrectedNormals}.Result
   WorldPositionOffset
   Metallic=StaticSwitch(True=LinearInterpolate(A=@n12, B=@n9.G, Alpha=@n6.R), False=@n12, Value=@n7)
  ) {attributeSetTypes [
    #1
    #2
    #3
    #4
    #5
    #6
    #7
   ]}
  FunctionOutput(None=@n14) {outputName=Out}
  n15 = NamedRerouteUsage()
  n16 = NamedRerouteUsage()
  n17 = ComponentMask(None=NamedRerouteUsage()) {b=true}
  n18 = ScalarParameter() {defaultValue=125, group=BugGoo, parameterName=OuterColor_TransitionStart}
  n19 = Saturate(None=Multiply(A=Subtract(A=@n17, B=@n18) {constB=125}, B=ScalarParameter() {defaultValue=0.015, group=BugGoo, parameterName=OuterColor_TransitionScale}) {constB=0.015})
  n20 = LinearInterpolate(
   A=LinearInterpolate(A=NamedRerouteUsage(), B=LinearInterpolate(A=@n15, B=@n16, Alpha=@n19), Alpha=NamedRerouteUsage())
   B=NamedRerouteUsage()
   Alpha=NamedRerouteUsage()
  )
  LinearInterpolate(
   A=@n20
   B=VectorParameter() {group=Bug, parameterName=InnerFresnelColor, defaultValue=<rgba 0.0037 0.0684 0.2344 1>}.RGB
   Alpha=NamedRerouteUsage()
  )
  LocalPosition()
  n21 = ComponentMask(None=NamedRerouteUsage()) {b=true}
  Multiply(A=@n3, B=@n21)
  n22 = Divide(A=MaterialFunctionCall(f3=MaterialFunctionCall() {materialFunction=&ObjectPivotPoint}.Object Pivot Location) {materialFunction=&AddComponents}.f3) {constB=128}
  n23 = MaterialFunctionCall() {materialFunction=&DitherTemporalAA}
  Multiply(A=Frac(None=Add(A=@n22, B=@n23.Result))) {constB=5}
  n24 = ComponentMask(None=AtmosphericLightVector()) {b=true, g=true, r=true}
  n25 = VertexNormalWS()
  n26 = Multiply(A=Multiply(A=DotProduct(A=@n24, B=@n25)) {constB=2}, B=@n25)
  n27 = DotProduct(
   A=Normalize(VectorInput=Subtract(A=@n26, B=@n24))
   B=Multiply(A=MaterialFunctionCall() {materialFunction=&CameraDirectionVector}.Result) {constB=-1}
  )
  n28 = Multiply(
   A=Power(Base=Max(A=@n27) {constB=0}, Exp=ScalarParameter() {defaultValue=4, group=BugGoo, parameterName=InnerScatteringPower})
   B=ScalarParameter() {defaultValue=0.4, group=BugGoo, parameterName=InnerScatteringScale}
  ) {constB=0.4}
  NamedRerouteDeclaration(None=Saturate(None=Saturate(None=@n28))) {name=ScatteringMask, nodeColor=<rgba 0 1 0.9297 1>}
  n29 = NamedRerouteUsage()
  n30 = VertexColor()
  n31 = AppendVector(A=@n30, B=@n30)
  n32 = VectorParameter() {group=BugGoo, parameterName=SHMax, defaultValue=<rgba 1.25 0.13 0.18 0.13>}
  n33 = VectorParameter() {group=BugGoo, parameterName=SHMin, defaultValue=<rgba 1 -0.16 -0.17 -0.15>}
  n34 = AppendVector(A=@n33.RGB, B=@n33.RGB)
  n35 = Subtract(A=AppendVector(A=@n32.RGB, B=@n32.RGB), B=@n34)
  n36 = Multiply(A=MaterialFunctionCall(X=@n30.G, Y=@n30.G, Z=@n30.G) {materialFunction=&MakeFloat3}.Result)
  MaterialExpressionStaticSwitchParameter_1 = StaticSwitchParameter(True=Add(A=Multiply(A=@n31, B=@n35), B=@n34), False=AppendVector(A=Add(A=@n30.R), B=Add(A=@n36) {constB=-0.5})) {group=BugGoo, parameterName=UseCustomSH Fit}
  n37 = ComponentMask(None=MaterialFunctionCall() {materialFunction=&CameraDirectionVector}.Result) {b=true, desc=NormalDistortion, g=true, r=true}
  n38 = Transform(None=NamedRerouteUsage()) {transformSourceType=TRANSFORMSOURCE_Local}
  n39 = Multiply(A=Multiply(A=NamedRerouteUsage(), B=NamedRerouteUsage()), B=ScalarParameter() {defaultValue=0.25, group=BugGoo, parameterName=SHBubbleScale})
  MaterialExpressionTransform_0 = Transform(None=Normalize(VectorInput=LinearInterpolate(A=@n37, B=@n38, Alpha=@n39))) {transformSourceType=TRANSFORMSOURCE_World, transformType=TRANSFORM_Local}
  n40 = Custom!(SH=@MaterialExpressionStaticSwitchParameter_1, Direction=@MaterialExpressionTransform_0) {
   code="//SH decode,  Direction = LightDir, SH = 2band coefficients. (encoded in Y-up)\r\nDirection = normalize(Direction);\r\nfloat4 shEvalDirection;\r\nshEvalDirection.x = 0.282095;\r\nshEvalDirection.y =0.488603 * Direction.z;\r\nshEvalDirection.z = 0.488603 * Direction.y;\r\nshEvalDirection.w =-0.488603 * Direction.x;\r\nreturn saturate(dot(SH, shEvalDirection));"
   description=SphericalHarmonicsEval
   outputType=CMOT_Float1
   inputs [
    {inputName=SH, input {expression=&self:MaterialExpressionStaticSwitchParameter_1}}
    {inputName=Direction, input {expression=&self:MaterialExpressionTransform_0}}
   ]
  }
  n41 = MaterialFunctionCall(
   Input=@n40
   Input Low=ScalarParameter() {group=BugGoo, parameterName=SHContrast_LowValue}
   Input High=ScalarParameter() {defaultValue=1, group=BugGoo, parameterName=SHContrast_HighValue}
   Target Low=Constant()
   Target High=Constant() {r=1}
  ) {materialFunction=&RemapValueRange}
  n42 = Power(Base=Saturate(None=OneMinus(None=@n41.Result)), Exp=ScalarParameter() {defaultValue=1, group=BugGoo, parameterName=SHPow})
  n43 = Fresnel(Normal=VertexNormalWS()) {baseReflectFraction=0, exponent=1}
  n44 = Subtract(A=NamedRerouteDeclaration(None=Multiply(A=@n42, B=@n43)) {name=Slurp1, nodeColor=<rgba 0.3125 1 0 1>}, B=NamedRerouteUsage())
  n45 = OneMinus(None=NamedRerouteUsage())
  n46 = QualitySwitch(Default=Add(A=Add(A=@n44, B=@n29), B=Multiply(A=@n45) {constB=0.0333}), Low=@n44)
  NamedRerouteDeclaration(None=ShadingPathSwitch(Default=@n46, Deferred=@n46, Forward=@n44, Mobile=@n44)) {name=SH Mask, nodeColor=<rgba 0.2891 0 1 1>}
  n47 = ScalarParameter() {defaultValue=25, group=BugGoo, parameterName=BubbleSize_A}
  NamedRerouteDeclaration(None=@n47) {name=BubbleSize_A, nodeColor=<rgba 1 0.7969 0 1>}
  n48 = OneMinus(None=MaterialFunctionCall() {materialFunction=&Fresnel_Function}.Result)
  n49 = TextureObject() {samplerType=SAMPLERTYPE_Masks, texture=/Game/Characters/Player/Male/Large/Bodies/M_LRG_Slurp_Monster/Textures/T_SlurpBubble}
  n50 = Constant() {r=2}
  n51 = MaterialFunctionCall(
   TextureObject=@n49
   TextureSize=@n47
   WorldPosition=NamedRerouteUsage()
   In=NamedRerouteUsage()
   World Space Normal=NamedRerouteUsage()
   ProjectionTransitionContrast=@n50
  ) {materialFunction=&WorldAlignedTexture_ScalePan}
  n52 = ComponentMask(None=@n51.XY Texture) {r=true}
  n53 = NamedRerouteUsage()
  n54 = MaterialFunctionCall!(
   WorldPosition=NamedRerouteUsage()
   World Space Normal=NamedRerouteUsage()
   UVSize_Front=@n53
   UVSize_Side=@n53
   UVSize_Top=@n53
   ProjectionTransitionContrast=Constant() {r=0.1}
  ) {materialFunction=/Game/Packages/Fortress_SharedMaterials/Base_Material_Functions/WorldAlignedTexture_UVsMasks}
  n55 = Saturate(None=Subtract(A=@n52, B=@n54.Mask_Top) {constB=0})
  n56 = MaterialFunctionCall(
   TextureObject=@n49
   TextureSize=ScalarParameter() {defaultValue=30, group=BugGoo, parameterName=BubbleSize_B}
   WorldPosition=NamedRerouteUsage()
   In=NamedRerouteUsage()
   World Space Normal=NamedRerouteUsage()
   ProjectionTransitionContrast=@n50
  ) {materialFunction=&WorldAlignedTexture_ScalePan}
  n57 = Saturate(None=Subtract(A=ComponentMask(None=@n56.XY Texture) {g=true}, B=@n54.Mask_Top))
  n58 = Add(A=@n55, B=@n57)
  NamedRerouteDeclaration(None=Saturate(None=Multiply(A=@n48, B=@n58))) {name=FresneledBubble_Biplanar_Mask, nodeColor=<rgba 0.4766 0 1 1>}
  n59 = NamedRerouteUsage()
  n60 = NamedRerouteUsage()
  n61 = Divide(B=Multiply(A=Abs(None=@n60)) {constB=-1})
  n62 = Time()
  n63 = NamedRerouteUsage()
  n64 = TextureObject() {samplerType=SAMPLERTYPE_Normal, texture=/Game/Characters/Player/Male/Large/Bodies/M_LRG_Slurp_Monster/Textures/T_SlurpBubble_N}
  n65 = TextureSample(UVs=Panner(Coordinate=ComponentMask(None=@n61) {b=true, r=true}, Time=@n62, Speed=@n63), Tex=@n64) {samplerSource=SSM_Wrap_WorldGroupSettings, samplerType=SAMPLERTYPE_Normal}
  n66 = Multiply(
   A=MaterialFunctionCall(X=@n65.R, Z=@n65.R) {materialFunction=&MakeFloat3}.Result
   B=ComponentMask(None=NamedRerouteUsage()) {g=true}
  )
  n67 = TextureSample(UVs=Panner(Coordinate=ComponentMask(None=@n61) {b=true, g=true}, Time=@n62, Speed=@n63), Tex=@n64) {samplerSource=SSM_Wrap_WorldGroupSettings, samplerType=SAMPLERTYPE_Normal}
  n68 = Multiply(
   A=MaterialFunctionCall(Y=@n67.R, Z=@n67.R) {materialFunction=&MakeFloat3}.Result
   B=ComponentMask(None=NamedRerouteUsage()) {r=true}
  )
  n69 = NamedRerouteUsage()
  n70 = Divide(A=NamedRerouteUsage(), B=Multiply(A=Abs(None=@n69)) {constB=-1})
  n71 = NamedRerouteUsage()
  n72 = TextureSample(UVs=Panner(Coordinate=ComponentMask(None=@n70) {b=true, r=true}, Time=@n62, Speed=@n71), Tex=@n64) {samplerSource=SSM_Wrap_WorldGroupSettings, samplerType=SAMPLERTYPE_Normal}
  n73 = Multiply(
   A=MaterialFunctionCall(X=@n72.R, Z=@n72.R) {materialFunction=&MakeFloat3}.Result
   B=ComponentMask(None=NamedRerouteUsage()) {g=true}
  )
  n74 = TextureSample(UVs=Panner(Coordinate=ComponentMask(None=@n70) {b=true, g=true}, Time=@n62, Speed=@n71), Tex=@n64) {samplerSource=SSM_Wrap_WorldGroupSettings, samplerType=SAMPLERTYPE_Normal}
  n75 = Multiply(
   A=MaterialFunctionCall(Y=@n74.R, Z=@n74.R) {materialFunction=&MakeFloat3}.Result
   B=ComponentMask(None=NamedRerouteUsage()) {r=true}
  )
  n76 = LinearInterpolate(A=Constant3Vector(), B=LinearInterpolate(A=Add(A=@n66, B=@n68), B=Add(A=@n73, B=@n75), Alpha=NamedRerouteUsage()), Alpha=NamedRerouteUsage())
  NamedRerouteDeclaration(None=Normalize(VectorInput=Add(A=@n59, B=@n76))) {name=Normal, nodeColor=<rgba 0.25 0.25 1 1>}
  n77 = Multiply(A=@n54.UV_Front) {constB=0.25}
  n78 = ComponentMask(None=@n77) {g=true}
  n79 = Multiply(A=Time()) {constB=0.05}
  n80 = TextureSample(UVs=AppendVector(A=ComponentMask(None=@n77) {r=true}, B=Add(A=@n78, B=@n79))) {
   constMipValue=0
   samplerSource=SSM_Wrap_WorldGroupSettings
   samplerType=SAMPLERTYPE_Grayscale
   texture=&LowResBlurredNoise
  }
  n81 = Multiply(A=@n54.UV_Side) {constB=0.25}
  n82 = ComponentMask(None=@n81) {g=true}
  n83 = TextureSample(UVs=AppendVector(A=ComponentMask(None=@n81) {r=true}, B=Add(A=@n82, B=@n79))) {
   constMipValue=0
   samplerSource=SSM_Wrap_WorldGroupSettings
   samplerType=SAMPLERTYPE_Grayscale
   texture=&LowResBlurredNoise
  }
  NamedRerouteDeclaration(None=Multiply(A=LinearInterpolate(A=@n80.R, B=@n83.R, Alpha=@n54.Mask_Side)) {constB=2}) {name=Noise_Biplanar_Mask, nodeColor=<rgba 1 0 0.3281 1>}
  n84 = ComponentMask(None=Saturate(None=MaterialFunctionCall() {materialFunction=/Engine/Functions/Engine_MaterialFunctions02/Texturing/3DSandMayaUVCoordinates}.UVChannel1)) {r=true}
  NamedRerouteDeclaration(None=@n84) {name=MayaUVLayoutR, nodeColor=<rgba 0.9922 1 0 1>}
  n85 = Add(A=@n55, B=@n57)
  n86 = MaterialFunctionCall(
   TextureObject=@n49
   TextureSize=ScalarParameter() {defaultValue=15, group=BugGoo, parameterName=BubbleSize_C}
   WorldPosition=NamedRerouteUsage()
   In=AppendVector(
    A=ScalarParameter() {group=BugGoo, parameterName=Bubble C Speed X}
    B=ScalarParameter() {defaultValue=0.12, group=BugGoo, parameterName=Bubble C Speed Y}
   )
   World Space Normal=NamedRerouteUsage()
   ProjectionTransitionContrast=@n50
  ) {materialFunction=&WorldAlignedTexture_ScalePan}
  n87 = Saturate(None=Subtract(A=ComponentMask(None=@n86.XY Texture) {b=true}, B=@n54.Mask_Top))
  n88 = QualitySwitch(Default=Add(A=@n85, B=Multiply(A=@n87) {constB=0.5}), Low=@n85)
  NamedRerouteDeclaration(None=ShadingPathSwitch(Default=@n88, Deferred=@n88, Forward=@n85, Mobile=@n85)) {name=Bubble_Triplanar_Mask, nodeColor=<rgba 0.9688 1 0 1>}
  NamedRerouteDeclaration(None=@n85) {name=Bubble_Biplanar_Mask, nodeColor=<rgba 0.9688 1 0 1>}
  n89 = Multiply(
   A=MaterialFunctionCall!(ReflectionVector, ReflectionTexture) {materialFunction=/Engine/Functions/Engine_MaterialFunctions01/Reflections/ViewAlignedReflection}.UVs
   B=ScalarParameter() {defaultValue=1, group=BugGoo, parameterName=ViewDependantDistortion}
  )
  n90 = AppendVector(
   A=ScalarParameter() {defaultValue=0.025, group=BugGoo, parameterName=water_SpeedX}
   B=ScalarParameter() {defaultValue=0.05, group=BugGoo, parameterName=water_SpeedY}
  )
  n91 = TextureSample(UVs=Rotator(Coordinate=Panner(Coordinate=@n89, Speed=@n90), Time=Time()) {speed=0.0001}) {
   constMipValue=0
   samplerSource=SSM_Wrap_WorldGroupSettings
   samplerType=SAMPLERTYPE_Grayscale
   texture=&LowResBlurredNoise
  }
  n92 = NamedRerouteUsage()
  n93 = Multiply(A=Power(Base=Saturate(None=@n92)) {constExponent=4}, B=NamedRerouteUsage())
  n94 = Fresnel() {baseReflectFraction=0, exponent=1}
  NamedRerouteDeclaration(None=Multiply(A=Multiply(A=@n91.G, B=@n93), B=OneMinus(None=@n94))) {name=Inner Bubble, nodeColor=<rgba 0.6641 1 0 1>}
  n95 = TextureSampleParameter2D() {parameterName=WPO_Mask, samplerType=SAMPLERTYPE_Masks, texture=&T_WhiteMask}
  n96 = WorldPosition() {worldPositionShaderOffset=WPT_ExcludeAllShaderOffsets}
  n97 = Subtract(A=Frac(None=Multiply(A=@n96.XYZ) {constB=0.0025})) {constB=0.5}
  n98 = MaterialFunctionCall(Float3=Abs(None=Multiply(A=@n97) {constB=2})) {materialFunction=&BreakOutFloat3Components}
  n99 = Add(A=Add(A=@n98.R, B=@n98.R), B=@n98.B)
  n100 = Multiply(A=Time()) {constB=0.25}
  n101 = Sine(None=Subtract(A=Subtract(A=@n99) {constB=2}, B=Frac(None=@n100)))
  n102 = Multiply(A=Multiply(A=Add(A=@n101)) {constB=0.5}, B=ScalarParameter() {defaultValue=1, group=BugGoo, parameterName=WPO_WorldNoiseScale})
  n103 = MaterialFunctionCall(Float3=Multiply(A=NamedRerouteUsage()) {constB=0.02}) {materialFunction=&BreakOutFloat3Components}
  n104 = Add(A=@n103.R, B=@n103.R)
  n105 = Subtract(A=Time(), B=ScalarParameter() {group=BugGoo, parameterName=Slurp_WPO_random_seed})
  n106 = Subtract(A=Subtract(A=Add(A=@n104, B=@n103.B)) {constB=2}, B=Frac(None=Multiply(A=@n105) {constB=0.6}))
  n107 = Multiply(A=Add(A=Sine(None=@n106))) {constB=0.5}
  n108 = Multiply(A=Multiply(A=VertexNormalWS(), B=Add(A=@n102, B=@n107)), B=@n84)
  NamedRerouteDeclaration(None=Multiply(A=Multiply(A=@n95.R, B=@n108), B=ScalarParameter() {defaultValue=1, group=BugGoo, parameterName=WPO_FinalScale})) {name=WPO, nodeColor=<rgba 1 0 0.1172 1>}
  n109 = ComponentMask(None=AtmosphericLightVector()) {b=true, g=true, r=true}
  n110 = Transform(None=NamedRerouteUsage()) {transformSourceType=TRANSFORMSOURCE_Local}
  n111 = Multiply(A=Multiply(A=DotProduct(A=@n109, B=@n110)) {constB=2}, B=@n110)
  n112 = DotProduct(
   A=Normalize(VectorInput=Subtract(A=@n111, B=@n109))
   B=Multiply(A=MaterialFunctionCall() {materialFunction=&CameraDirectionVector}.Result)
  )
  n113 = Max(A=@n112) {constB=0}
  n114 = Constant() {r=5}
  n115 = Saturate(None=Multiply(A=Power(Base=@n113, Exp=@n114), B=ScalarParameter() {defaultValue=0.15, group=BugGoo, parameterName=Bubble_BackSpecular_Scale}) {constB=0.15})
  n116 = Multiply(A=@n112) {constB=-1}
  n117 = Constant() {r=50}
  n118 = Saturate(None=Multiply(A=Power(Base=@n116, Exp=@n117), B=ScalarParameter() {defaultValue=0.55, group=BugGoo, parameterName=Bubble_HighlightSpecular_Scale}) {constB=0.55})
  NamedRerouteDeclaration(None=Multiply(A=Add(A=@n115, B=@n118), B=NamedRerouteUsage())) {name=Bubble_Specular, nodeColor=<rgba 0.2656 0 1 1>}
  NamedRerouteDeclaration(None=VertexInterpolator(VS=LocalPosition() {localOrigin=InstancePreSkinning}.XYZ).PS) {name=LocalPosition, nodeColor=<rgba 0.6875 0 1 1>}
  n119 = Constant() {r=0.66}
  n120 = Constant() {r=1}
  n121 = Constant() {r=0.66}
  n122 = Constant() {r=0.5}
  n123 = Add(A=@n20, B=NamedRerouteUsage())
  n124 = NamedRerouteUsage()
  n125 = NamedRerouteUsage()
  NamedRerouteDeclaration(None=Multiply(
    A=MaterialFunctionCall(Value_Morning=@n119, Value_Day=@n120, Value_Evening=@n121, Value_Night=@n122) {materialFunction=&Scalar_Time_of_Day}.Result
    B=LinearInterpolate(A=@n123, B=@n124, Alpha=@n125)
   )) {name=EmissiveColor, nodeColor=<rgba 1 0.2812 0 1>}
  n126 = NamedRerouteUsage()
  n127 = ScalarParameter() {defaultValue=0.5, group=BugGoo, parameterName=Blend Sharpness, sliderMax=1}
  n128 = Max(A=Add(A=Abs(None=@n126), B=Multiply(A=@n127) {constB=-0.55}) {constB=-0.55}) {constB=0}
  n129 = Constant3Vector() {constant=<rgba 1 1 1 0>}
  NamedRerouteDeclaration(None=Divide(A=@n128, B=DotProduct(A=@n128, B=@n129))) {name=BubbleNormalBlend_Mask, nodeColor=<rgba 0.2188 1 0 1>}
  n130 = ScalarParameter() {desc="0 - Pirple\r\n1 - Blue\r\n2 - Green\r\n3 - Yellow", group=BugGoo, parameterName=BugGooColorScheme, sortPriority=0}
  MaterialExpressionVectorParameter_1 = VectorParameter() {group=BugGoo, parameterName=Purple_InnerColor, sortPriority=10, defaultValue=<rgba 0.1344 0.0985 0.5729 1>}
  MaterialExpressionVectorParameter_2 = VectorParameter() {group=BugGoo, parameterName=Blue_InnerColor, sortPriority=10, defaultValue=<rgba 0.0615 0.2457 1 1>}
  MaterialExpressionVectorParameter_7 = VectorParameter() {group=BugGoo, parameterName=Green_InnerColor, sortPriority=10, defaultValue=<rgba 0.25 0.1866 0 1>}
  MaterialExpressionVectorParameter_8 = VectorParameter() {group=BugGoo, parameterName=Yellow_InnerColor, sortPriority=10, defaultValue=<rgba 1 0.1017 0.0065 1>}
  n131 = Switch!(
   SwitchValue=@n130
   Default=@MaterialExpressionVectorParameter_1.RGB
   Purple=@MaterialExpressionVectorParameter_1.RGB
   Blue=@MaterialExpressionVectorParameter_2.RGB
   Green=@MaterialExpressionVectorParameter_7.RGB
   Yellow=@MaterialExpressionVectorParameter_8.RGB
  ) {inputs [
    {
     inputName=Purple
     input {
      expression=&self:MaterialExpressionVectorParameter_1
      inputName=None
      mask=1
      maskA=0
      maskB=1
      maskG=1
      maskR=1
      outputIndex=0
     }
    }
    {
     inputName=Blue
     input {
      expression=&self:MaterialExpressionVectorParameter_2
      inputName=None
      mask=1
      maskA=0
      maskB=1
      maskG=1
      maskR=1
      outputIndex=0
     }
    }
    {
     inputName=Green
     input {
      expression=&self:MaterialExpressionVectorParameter_7
      inputName=None
      mask=1
      maskA=0
      maskB=1
      maskG=1
      maskR=1
      outputIndex=0
     }
    }
    {
     inputName=Yellow
     input {
      expression=&self:MaterialExpressionVectorParameter_8
      inputName=None
      mask=1
      maskA=0
      maskB=1
      maskG=1
      maskR=1
      outputIndex=0
     }
    }
   ]}
  NamedRerouteDeclaration(None=@n131) {name=InnerColor, nodeColor=<rgba 1 0.7527 0.0309 1>}
  MaterialExpressionVectorParameter_11 = VectorParameter() {group=BugGoo, parameterName=Purple_OuterColorA, sortPriority=10, defaultValue=<rgba 0.758 0.2643 0.901 1>}
  MaterialExpressionVectorParameter_12 = VectorParameter() {group=BugGoo, parameterName=Blue_OuterColorA, sortPriority=10, defaultValue=<rgba 0.15 0.6341 0.8906 1>}
  MaterialExpressionVectorParameter_13 = VectorParameter() {group=BugGoo, parameterName=Green_OuterColorA, sortPriority=10, defaultValue=<rgba 0.3002 0.724 0 1>}
  MaterialExpressionVectorParameter_14 = VectorParameter() {group=BugGoo, parameterName=Yellow_OuterColorA, sortPriority=10, defaultValue=<rgba 1 0.8466 0 1>}
  n132 = Switch(
   SwitchValue=@n130
   Default=@MaterialExpressionVectorParameter_11.RGB
   Purple=@MaterialExpressionVectorParameter_11.RGB
   Blue=@MaterialExpressionVectorParameter_12.RGB
   Green=@MaterialExpressionVectorParameter_13.RGB
   Yellow=@MaterialExpressionVectorParameter_14.RGB
  ) {inputs [
    {
     inputName=Purple
     input {
      expression=&self:MaterialExpressionVectorParameter_11
      inputName=None
      mask=1
      maskA=0
      maskB=1
      maskG=1
      maskR=1
      outputIndex=0
     }
    }
    {
     inputName=Blue
     input {
      expression=&self:MaterialExpressionVectorParameter_12
      inputName=None
      mask=1
      maskA=0
      maskB=1
      maskG=1
      maskR=1
      outputIndex=0
     }
    }
    {
     inputName=Green
     input {
      expression=&self:MaterialExpressionVectorParameter_13
      inputName=None
      mask=1
      maskA=0
      maskB=1
      maskG=1
      maskR=1
      outputIndex=0
     }
    }
    {
     inputName=Yellow
     input {
      expression=&self:MaterialExpressionVectorParameter_14
      inputName=None
      mask=1
      maskA=0
      maskB=1
      maskG=1
      maskR=1
      outputIndex=0
     }
    }
   ]}
  NamedRerouteDeclaration(None=@n132) {name=OuterColorA, nodeColor=<rgba 1 0.7527 0.0309 1>}
  MaterialExpressionVectorParameter_15 = VectorParameter() {group=BugGoo, parameterName=Purple_OuterColorB, sortPriority=10, defaultValue=<rgba 0.2642 0.0761 0.6354 1>}
  MaterialExpressionVectorParameter_16 = VectorParameter() {group=BugGoo, parameterName=Blue_OuterColorB, sortPriority=10, defaultValue=<rgba 0.0859 0.2606 0.75 1>}
  MaterialExpressionVectorParameter_17 = VectorParameter() {group=BugGoo, parameterName=Green_OuterColorB, sortPriority=10, defaultValue=<rgba 0.6425 0.75 0.1392 1>}
  MaterialExpressionVectorParameter_18 = VectorParameter() {group=BugGoo, parameterName=Yellow_OuterColorB, sortPriority=10, defaultValue=<rgba 0.9896 0.9126 0.3505 1>}
  n133 = Switch(
   SwitchValue=@n130
   Default=@MaterialExpressionVectorParameter_15.RGB
   Purple=@MaterialExpressionVectorParameter_15.RGB
   Blue=@MaterialExpressionVectorParameter_16.RGB
   Green=@MaterialExpressionVectorParameter_17.RGB
   Yellow=@MaterialExpressionVectorParameter_18.RGB
  ) {inputs [
    {
     inputName=Purple
     input {
      expression=&self:MaterialExpressionVectorParameter_15
      inputName=None
      mask=1
      maskA=0
      maskB=1
      maskG=1
      maskR=1
      outputIndex=0
     }
    }
    {
     inputName=Blue
     input {
      expression=&self:MaterialExpressionVectorParameter_16
      inputName=None
      mask=1
      maskA=0
      maskB=1
      maskG=1
      maskR=1
      outputIndex=0
     }
    }
    {
     inputName=Green
     input {
      expression=&self:MaterialExpressionVectorParameter_17
      inputName=None
      mask=1
      maskA=0
      maskB=1
      maskG=1
      maskR=1
      outputIndex=0
     }
    }
    {
     inputName=Yellow
     input {
      expression=&self:MaterialExpressionVectorParameter_18
      inputName=None
      mask=1
      maskA=0
      maskB=1
      maskG=1
      maskR=1
      outputIndex=0
     }
    }
   ]}
  NamedRerouteDeclaration(None=@n133) {name=OuterColorB, nodeColor=<rgba 1 0.7527 0.0309 1>}
  n134 = Multiply(A=ScalarParameter() {defaultValue=1, group=BugGoo, parameterName=AnimCausticsMulti}, B=Add(A=@n3, B=@n21))
  NamedRerouteDeclaration(None=@n134) {name=Caustic, nodeColor=<rgba 0 1 0.8125 1>}
  MaterialExpressionVectorParameter_19 = VectorParameter() {group=BugGoo, parameterName=Purple_InnerScatteringColor, sortPriority=10, defaultValue=<rgba 0.5963 0.2385 0.7917 1>}
  MaterialExpressionVectorParameter_20 = VectorParameter() {group=BugGoo, parameterName=Blue_InnerScatteringColor, sortPriority=10, defaultValue=<rgba 0.1692 0.5729 0.551 1>}
  MaterialExpressionVectorParameter_21 = VectorParameter() {group=BugGoo, parameterName=Green_InnerScatteringColor, sortPriority=10, defaultValue=<rgba 0.5573 0.5353 0.0946 1>}
  MaterialExpressionVectorParameter_22 = VectorParameter() {group=BugGoo, parameterName=Yellow_InnerScatteringColor, sortPriority=10, defaultValue=<rgba 1 0.3937 0.1562 1>}
  n135 = Switch(
   SwitchValue=@n130
   Default=@MaterialExpressionVectorParameter_19.RGB
   Purple=@MaterialExpressionVectorParameter_19.RGB
   Blue=@MaterialExpressionVectorParameter_20.RGB
   Green=@MaterialExpressionVectorParameter_21.RGB
   Yellow=@MaterialExpressionVectorParameter_22.RGB
  ) {inputs [
    {
     inputName=Purple
     input {
      expression=&self:MaterialExpressionVectorParameter_19
      inputName=None
      mask=1
      maskA=0
      maskB=1
      maskG=1
      maskR=1
      outputIndex=0
     }
    }
    {
     inputName=Blue
     input {
      expression=&self:MaterialExpressionVectorParameter_20
      inputName=None
      mask=1
      maskA=0
      maskB=1
      maskG=1
      maskR=1
      outputIndex=0
     }
    }
    {
     inputName=Green
     input {
      expression=&self:MaterialExpressionVectorParameter_21
      inputName=None
      mask=1
      maskA=0
      maskB=1
      maskG=1
      maskR=1
      outputIndex=0
     }
    }
    {
     inputName=Yellow
     input {
      expression=&self:MaterialExpressionVectorParameter_22
      inputName=None
      mask=1
      maskA=0
      maskB=1
      maskG=1
      maskR=1
      outputIndex=0
     }
    }
   ]}
  NamedRerouteDeclaration(None=@n135) {name=InnerScatteringColor, nodeColor=<rgba 1 0.7527 0.0309 1>}
  NamedRerouteDeclaration(None=ScalarParameter() {defaultValue=0.1, group=Bug, parameterName=HeightRatio}) {name=HeightRatio, nodeColor=<rgba 1 0 0.4688 1>}
  NamedRerouteDeclaration(None=ScalarParameter() {group=Bug, parameterName=ReferencePlane}) {name=ReferencePlane, nodeColor=<rgba 0 0.7891 1 1>}
  n136 = VectorParameter() {group=Bug, parameterName=Particle_Tiling, sortPriority=4, defaultValue=<rgba 10 10 0 1>}
  NamedRerouteDeclaration(None=Multiply(A=TextureCoordinate(), B=ComponentMask(None=@n136.RGB) {g=true, r=true}) {constB=20}) {name=UVs, nodeColor=<rgba 0.1484 1 0 1>}
  n137 = Add(A=@n52, B=ComponentMask(None=@n51.XY Texture) {g=true})
  n138 = ComponentMask(None=@n51.XY Texture) {b=true}
  NamedRerouteDeclaration(None=Saturate(None=Add(A=@n137, B=@n138))) {name=Bubble_A_Mask, nodeColor=<rgba 0 0.0625 1 1>}
  NamedRerouteDeclaration(None=TextureObject() {texture=/KetchupChips/Assets/Textures/T_Goo_PlayerBody_TA}) {name=NoiseTex, nodeColor=<rgba 0.1719 1 0 1>}
  n139 = Multiply(A=Time(), B=ScalarParameter() {defaultValue=0.2, group=Bug, parameterName=Particle_BlinkSpeed}) {constB=0.1}
  n140 = MaterialFunctionCall(Float2=MaterialFunctionCall!(
    UV Coord=NamedRerouteUsage()
    Height=ScalarParameter() {defaultValue=2, group=Bug, parameterName=Particles_Height_Inner, sortPriority=50}
    ReferencePlane=NamedRerouteUsage()
    HeightRatio=NamedRerouteUsage()
    Normal
   ) {materialFunction=/Game/Global/MaterialFunctions/MF_BumpOffset_Accurate}.Result) {materialFunction=&BreakOutFloat2Components}
  n141 = Frac(None=Multiply(A=Time(), B=ScalarParameter() {defaultValue=0.035, group=Bug, parameterName=Particle_PanSpeed}) {constB=0.1})
  n142 = TextureSample(
   UVs=MaterialFunctionCall(X=@n140.R, Y=Add(A=@n140.G, B=@n141)) {materialFunction=&MakeFloat2}.Result
   Tex=NamedRerouteUsage()
  ) {samplerSource=SSM_Wrap_WorldGroupSettings, samplerType=SAMPLERTYPE_Masks}
  n143 = Sine(None=Add(A=@n139, B=@n142.R))
  n144 = ScalarParameter() {defaultValue=1, group=Bug, parameterName=Particles_Opacity_Inner}
  n145 = Multiply(A=NamedRerouteUsage()) {constB=0.4}
  n146 = Constant2Vector() {g=0.138, r=0.4165}
  n147 = MaterialFunctionCall(Float2=MaterialFunctionCall(
    UV Coord=Add(A=@n145, B=@n146) {constB=0.1}
    Height=ScalarParameter() {defaultValue=1, group=Bug, parameterName=Particles_Height_Outter, sortPriority=50}
    ReferencePlane=NamedRerouteUsage()
    HeightRatio=NamedRerouteUsage()
   ) {materialFunction=&MF_BumpOffset_Accurate}.Result) {materialFunction=&BreakOutFloat2Components}
  n148 = TextureSample(
   UVs=MaterialFunctionCall(X=@n147.R, Y=Add(A=@n141, B=@n147.G)) {materialFunction=&MakeFloat2}.Result
   Tex=NamedRerouteUsage()
  ) {samplerSource=SSM_Wrap_WorldGroupSettings, samplerType=SAMPLERTYPE_Masks}
  n149 = Sine(None=Add(A=Add(A=@n139) {constB=0.3}, B=@n148.R))
  n150 = Add(
   A=Multiply(A=ConstantBiasScale(None=@n143), B=Multiply(A=@n142.G, B=@n144) {constB=0.5})
   B=Multiply(A=ConstantBiasScale(None=@n149), B=SmoothStep(Value=@n148.G) {constMax=0.75})
  )
  NamedRerouteDeclaration(None=Multiply(A=Saturate(None=@n150), B=NamedRerouteUsage())) {name=ParticleMask, nodeColor=<rgba 0.4297 1 0 1>}
  n151 = NamedRerouteUsage()
  n152 = ComponentMask(None=@n134) {b=true}
  NamedRerouteDeclaration(None=Multiply(A=Multiply(A=@n151, B=@n152), B=ScalarParameter() {defaultValue=-5, group=Bug, parameterName=Caustics_Intensity})) {name=Caustics, nodeColor=<rgba 1 0.5625 0 1>}
  MaterialExpressionVectorParameter_9 = VectorParameter() {group=BugGoo, parameterName=Purple_ParticleColor, sortPriority=10, defaultValue=<rgba 0.0639 0.2304 1 1>}
  MaterialExpressionVectorParameter_23 = VectorParameter() {group=BugGoo, parameterName=Blue_ParticleColor, sortPriority=10, defaultValue=<rgba 0.0639 0.2304 1 1>}
  MaterialExpressionVectorParameter_24 = VectorParameter() {group=BugGoo, parameterName=Green_ParticleColor, sortPriority=10, defaultValue=<rgba 0.0639 0.2304 1 1>}
  MaterialExpressionVectorParameter_25 = VectorParameter() {group=BugGoo, parameterName=Yellow_ParticleColor, sortPriority=10, defaultValue=<rgba 0.0639 0.2304 1 1>}
  n153 = Switch(
   SwitchValue=@n130
   Default=@MaterialExpressionVectorParameter_9.RGB
   Purple=@MaterialExpressionVectorParameter_9.RGB
   Blue=@MaterialExpressionVectorParameter_23.RGB
   Green=@MaterialExpressionVectorParameter_24.RGB
   Yellow=@MaterialExpressionVectorParameter_25.RGB
  ) {inputs [
    {
     inputName=Purple
     input {
      expression=&self:MaterialExpressionVectorParameter_9
      inputName=None
      mask=1
      maskA=0
      maskB=1
      maskG=1
      maskR=1
      outputIndex=0
     }
    }
    {
     inputName=Blue
     input {
      expression=&self:MaterialExpressionVectorParameter_23
      inputName=None
      mask=1
      maskA=0
      maskB=1
      maskG=1
      maskR=1
      outputIndex=0
     }
    }
    {
     inputName=Green
     input {
      expression=&self:MaterialExpressionVectorParameter_24
      inputName=None
      mask=1
      maskA=0
      maskB=1
      maskG=1
      maskR=1
      outputIndex=0
     }
    }
    {
     inputName=Yellow
     input {
      expression=&self:MaterialExpressionVectorParameter_25
      inputName=None
      mask=1
      maskA=0
      maskB=1
      maskG=1
      maskR=1
      outputIndex=0
     }
    }
   ]}
  NamedRerouteDeclaration(None=@n153) {name=ParticleColor, nodeColor=<rgba 1 0.7527 0.0309 1>}
  n154 = ScalarParameter() {defaultValue=4, group=BugGoo, parameterName=WaterTiling}
  n155 = Divide(A=NamedRerouteUsage(), B=ScalarParameter() {defaultValue=40, group=BugGoo, parameterName=Triplanner_ConstantScaleDivisor})
  n156 = Multiply(A=Time(), B=VectorParameter() {group=BugGoo, parameterName=Water Speed, defaultValue=<rgba 0 0 0.1 1>}.RGB)
  n157 = Time()
  n158 = VectorParameter() {group=BugGoo, parameterName=TriPlanner Speed, defaultValue=<rgba 0 0 0.05 1>}
  n159 = Add(
   A=Multiply(A=ScalarParameter() {defaultValue=8, group=BugGoo, parameterName=Distortion Tiling}, B=@n155)
   B=Frac(None=Multiply(A=@n157, B=@n158.RGB))
  )
  n160 = TextureObjectParameter() {
   group=BugGoo
   parameterName=EmissiveGradient
   samplerType=SAMPLERTYPE_Grayscale
   texture=&LowResBlurredNoise
  }
  n161 = TextureSample(UVs=ComponentMask(None=@n159) {b=true, g=true}, Tex=@n160) {samplerSource=SSM_Wrap_WorldGroupSettings, samplerType=SAMPLERTYPE_Masks}
  n162 = TextureSample(UVs=ComponentMask(None=@n159) {b=true, r=true}, Tex=@n160) {samplerSource=SSM_Wrap_WorldGroupSettings, samplerType=SAMPLERTYPE_Masks}
  n163 = Add(
   A=Add(A=Multiply(A=@n154, B=@n155), B=Frac(None=@n156))
   B=Multiply(A=LinearInterpolate(A=@n161.RGB, B=@n162.RGB), B=ScalarParameter() {defaultValue=0.005, group=BugGoo, parameterName=Offset Intensity}) {constB=0.01}
  )
  n164 = TextureObjectParameter() {
   group=BugGoo
   parameterName=WaterNormal
   samplerType=SAMPLERTYPE_Normal
   texture=/Engine/Functions/Engine_MaterialFunctions02/ExampleContent/Textures/water_n
  }
  n165 = TextureSample(UVs=ComponentMask(None=@n163) {b=true, r=true}, Tex=@n164) {
   samplerSource=SSM_Wrap_WorldGroupSettings
   samplerType=SAMPLERTYPE_Normal
   texture=/Game/Weapons/WeaponSkins/Textures/T_Swirly_Noise_highContrast
  }
  n166 = TextureSample(UVs=ComponentMask(None=@n163) {g=true, r=true}, Tex=@n164) {
   samplerSource=SSM_Wrap_WorldGroupSettings
   samplerType=SAMPLERTYPE_Normal
   texture=&T_Swirly_Noise_highContrast
  }
  n167 = ComponentMask(None=NamedRerouteUsage()) {b=true}
  n168 = ComponentMask(None=MaterialFunctionCall(In=Abs(None=@n167), Contrast=ScalarParameter() {defaultValue=0.2, group=BugGoo, parameterName=TriPlanner_AxisBlendStrength}) {materialFunction=&CheapContrast}.Result) {b=true}
  NamedRerouteDeclaration(None=MaterialFunctionCall(
    Normal=LinearInterpolate(A=@n165.RGB, B=@n166.RGB, Alpha=@n168)
    Flatness=ScalarParameter() {defaultValue=0.6, group=BugGoo, parameterName=WaterIntensity}
   ) {materialFunction=&FlattenNormal}.Result) {name=DetialNormal, nodeColor=<rgba 0.25 0.25 1 1>}
  NamedRerouteDeclaration(None=VertexInterpolator(VS=PreSkinnedNormal()).PS) {name=WorldSpaceNormal, nodeColor=<rgba 0.3359 0 1 1>}
  n169 = Fresnel(Normal=VertexNormalWS()) {baseReflectFraction=0, exponent=1}
  NamedRerouteDeclaration(None=OneMinus(None=@n169)) {name=InvertedFresnel, nodeColor=<rgba 1 0.3516 0 1>}
  NamedRerouteDeclaration(None=SmoothStep(Min=ScalarParameter() {defaultValue=0.7, parameterName=SS_MIN}, Max=ScalarParameter() {defaultValue=0.9, parameterName=SS_MAX}, Value=@n169) {constMin=0.7}) {name=SkinFresnel, nodeColor=<rgba 0 1 0.7188 1>}
  NamedRerouteDeclaration(None=SmoothStep(Value=@n169) {constMin=0.2}) {name=SkinFresnelDark, nodeColor=<rgba 0 0.4844 1 1>}
  n170 = NamedRerouteUsage()
  n171 = CameraPositionWS()
  n172 = Normalize(VectorInput=Subtract(A=NamedRerouteUsage(), B=TransformPosition(None=@n171) {transformSourceType=TRANSFORMPOSSOURCE_World}))
  n173 = QualitySwitch(Default=Add(A=@n170, B=Multiply(A=@n172) {constB=12}), Low=@n170)
  NamedRerouteDeclaration(None=ShadingPathSwitch(Default=@n173, Deferred=@n173, Forward=@n170, Mobile=@n170)) {name=Bubble_A_WorldPosition, nodeColor=<rgba 0.5703 0 1 1>}
  n174 = Multiply(A=@n172) {constB=15}
  n175 = Add(A=@n170) {constB=20}
  n176 = QualitySwitch(Default=Add(A=Add(A=@n170, B=@n174)) {constB=20}, Low=@n175)
  NamedRerouteDeclaration(None=ShadingPathSwitch(Default=@n176, Deferred=@n176, Forward=@n175, Mobile=@n175)) {name=Bubble_B_WorldPosition, nodeColor=<rgba 0.5703 0 1 1>}
  n177 = Multiply(A=@n172) {constB=15}
  n178 = Add(A=@n170) {constB=166}
  n179 = QualitySwitch(Default=Add(A=Add(A=@n170, B=@n177)) {constB=166}, Low=@n178)
  NamedRerouteDeclaration(None=ShadingPathSwitch(Default=@n179, Deferred=@n179, Forward=@n178, Mobile=@n178)) {name=Bubble_C_WorldPosition, nodeColor=<rgba 0.5703 0 1 1>}
  n180 = NamedRerouteUsage()
  n181 = Time()
  n182 = Multiply(A=Add(A=ComponentMask(None=@n180) {b=true}, B=Multiply(A=@n181) {constB=0.5})) {constB=0.05}
  n183 = Multiply(A=Sine(None=Frac(None=@n182))) {constB=0}
  NamedRerouteDeclaration(None=AppendVector(A=@n183, B=ScalarParameter() {defaultValue=0.025, group=BugGoo, parameterName=Bubble A Speed Y})) {name=Bubble_A_SpeedY, nodeColor=<rgba 0 0.9531 1 1>}
  NamedRerouteDeclaration(None=AppendVector(A=Multiply(A=@n183) {constB=-0.5}, B=ScalarParameter() {defaultValue=0.05, group=BugGoo, parameterName=Bubble B Speed Y})) {name=Bubble_B_SpeedY, nodeColor=<rgba 0 0.9531 1 1>}
  PixelNormalWS()
  PreSkinnedNormal()
  Saturate()
  VertexNormalWS()
  VertexNormalWS()
 ]}
