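# M_Goo_PlayerBody is a Default Lit material for the player body's goo look, flagged for use on Niagara mesh particles.
# It is a thin wrapper: its graph holds a single call to the MF_Goo_PlayerBody material function, which contains the shading logic, and uses Material Attributes.
# Automatic usage flagging in the editor is off, so Niagara mesh particles are the only usage it explicitly enables.
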Material /KetchupChips/Assets/Materials/M_Goo_PlayerBody
graph {
 properties {bAutomaticallySetUsageInEditor=false, bUseMaterialAttributes=true, bUsedWithNiagaraMeshParticles=true, usedShadingModels=Default Lit}
 nodes [MaterialFunctionCall() {materialFunction=&/MF_Goo_PlayerBody}]
}
